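# UI sound cue for the item shop, named for closing the About Refresh Timer element.
# A mixer layers a simple shop click with a medium toolkit whoosh, played together.
# Only the whoosh passes through a modulator, which lowers its volume and slightly randomizes its pitch and level.
# The cue uses the Fort_UI sound class, boosts overall volume, and overrides concurrency to prevent new instances at the limit.
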SoundCue /Game/Sounds/Fort_UI_Sounds/FrontEnd/ItemShop/Fort_UI_Shop_About_RefreshTimer_Close_Cue
root=Mixer_0
nodes [
 {id=Mixer_0, children [Modulator_2,WavePlayer_3], properties {inputVolume [1,1]}}
 {id=Modulator_2, children [WavePlayer_0], properties {pitchMax=1, pitchMin=0.9, volumeMax=0.5, volumeMin=0.4}}
 {
  id=WavePlayer_0
  properties {soundWaveAssetPtr=/Game/Sounds/Fort_UI_Sounds/FrontEnd/_GenericToolkit_/Whooshes/Misc_Whooshes/wav/UI_Toolkit_Whoosh_Med_10}
 }
 {id=WavePlayer_3, properties {soundWaveAssetPtr=/Game/Sounds/Fort_UI_Sounds/FrontEnd/ItemShop/wav/Fort_UI_Shop_Click_Simple}}
]
properties {
 bOverrideConcurrency=true
 maxDistance=2097152
 soundClassObject=&Fort_UI
 volumeMultiplier=1.5
 concurrencyOverrides {resolutionRule=PreventNew}
}
