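# UI sound cue for the item shop offer details window appearing, built from three layered whoosh waves and routed to the Fort_UI sound class.
# A mixer combines a follow-up modal whoosh, a briefly delayed medium modal whoosh and a menu-rotate whoosh with slightly randomized volume.
# An outer modulator lowers the mix, on top of a low cue volume multiplier.
# A concurrency override limits simultaneous plays and rejects new ones beyond the cap.
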SoundCue /Game/Sounds/Fort_UI_Sounds/FrontEnd/ItemShop/Fortnite_UI_Shop_OfferDetailsWindowAppear_Cue
root=Modulator_2
nodes [
 {id=Delay_0, children [WavePlayer_2], properties {delayMax=0.15, delayMin=0.15}}
 {id=Mixer_0, children [WavePlayer_1,Delay_0,Modulator_0], properties {inputVolume [1,1,1]}}
 {id=Modulator_0, children [WavePlayer_0], properties {volumeMax=0.9, volumeMin=0.8}}
 {id=Modulator_2, children [Mixer_0], properties {pitchMin=1, volumeMax=0.6, volumeMin=0.6}}
 {id=WavePlayer_0, properties {soundWaveAssetPtr=/Game/Sounds/Fort_UI_Sounds/FrontEnd/ItemShop/wav/Fort_UI_Whoosh_Menu_Rotate}}
 {id=WavePlayer_1, properties {soundWaveAssetPtr=/Game/Sounds/Fort_UI_Sounds/FrontEnd/BattlePass/BP_ToolKit/BP_Modal_Whoosh_FollowUp_01}}
 {id=WavePlayer_2, properties {soundWaveAssetPtr=/Game/Sounds/Fort_UI_Sounds/FrontEnd/BattlePass/BP_ToolKit/BP_Modal_Whoosh_Med_01}}
]
properties {
 bOverrideConcurrency=true
 maxDistance=2097152
 soundClassObject=&Fort_UI
 volumeMultiplier=0.1
 concurrencyOverrides {maxCount=4, resolutionRule=PreventNew, retriggerTime=0.1, volumeScale=0.8}
}
