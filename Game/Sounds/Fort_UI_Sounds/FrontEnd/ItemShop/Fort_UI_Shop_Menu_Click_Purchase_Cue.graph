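# Sound cue for the purchase click in the front-end Item Shop menu, in the Fort_UI sound class.
# A Mixer layers two waves at equal input volume, a purchase click and a deep impact, each through its own Modulator.
# The purchase click's Modulator randomizes its gain above unity, and the whole cue carries a volume boost.
# A concurrency override uses PreventNew, so new plays are refused once the voice limit is reached.
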SoundCue /Game/Sounds/Fort_UI_Sounds/FrontEnd/ItemShop/Fort_UI_Shop_Menu_Click_Purchase_Cue
root=Mixer_0
nodes [
 {id=Mixer_0, children [Modulator_0,Modulator_1], properties {inputVolume [1,1]}}
 {id=Modulator_0, children [WavePlayer_0], properties {volumeMax=1.5, volumeMin=1.2}}
 {id=Modulator_1, children [WavePlayer_1]}
 {id=WavePlayer_0, properties {soundWaveAssetPtr=/Game/Sounds/Fort_UI_Sounds/FrontEnd/ItemShop/wav/Fort_UI_Shop_Click_Purchase}}
 {id=WavePlayer_1, properties {soundWaveAssetPtr=/Game/Sounds/Fort_UI_Sounds/FrontEnd/ItemShop/wav/Fort_UI_Shop_Click_DeepImpact_01}}
]
properties {
 bOverrideConcurrency=true
 maxDistance=2097152
 soundClassObject=&Fort_UI
 volumeMultiplier=1.5
 concurrencyOverrides {resolutionRule=PreventNew}
}
