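# Sound cue for a simple click in the front-end Item Shop menu, variant B, routed to the Fort_UI sound class.
# It plays the shop MenuToggle wave through a Modulator that adds slight random pitch and volume variation.
# A concurrency override allows one instance at a time, stopping the oldest with a short release when retriggered.
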SoundCue /Game/Sounds/Fort_UI_Sounds/FrontEnd/ItemShop/Fort_UI_Shop_Menu_Click_Simple_B_Cue
root=Modulator_0
nodes [
 {id=Modulator_0, children [WavePlayer_1], properties {volumeMin=1}}
 {id=WavePlayer_1, properties {soundWaveAssetPtr=/Game/Sounds/Fort_UI_Sounds/FrontEnd/ItemShop/wav/Fort_UI_Shop_MenuToggle}}
]
properties {
 bOverrideConcurrency=true
 maxDistance=2097152
 soundClassObject=&Fort_UI
 volumeMultiplier=1
 concurrencyOverrides {maxCount=1, resolutionRule=StopOldest, voiceStealReleaseTime=0.2}
}
