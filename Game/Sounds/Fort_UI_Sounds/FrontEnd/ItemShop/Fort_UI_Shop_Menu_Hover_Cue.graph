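# Item shop menu hover sound cue for the front-end UI, in the Fort_UI sound class.
# A mixer layers a hover-click wave with a quieter button hover wave, feeding a root modulator whose pitch and volume ranges are fixed at unity, so playback has no variation.
# It references a shared front-end menu hover-tick concurrency asset, named for a one-voice, stop-oldest limit.
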SoundCue /Game/Sounds/Fort_UI_Sounds/FrontEnd/ItemShop/Fort_UI_Shop_Menu_Hover_Cue
root=Modulator_0
nodes [
 {id=Mixer_1, children [WavePlayer_3,WavePlayer_0], properties {inputVolume [0.75,0.5]}}
 {id=Modulator_0, children [Mixer_1], properties {pitchMax=1, pitchMin=1, volumeMax=1, volumeMin=1}}
 {id=WavePlayer_0, properties {soundWaveAssetPtr=/Game/Sounds/Fort_UI_Sounds/FrontEnd/New_FrontEnd/UI_FN_Button_Hover}}
 {id=WavePlayer_3, properties {soundWaveAssetPtr=/Game/Sounds/Fort_UI_Sounds/FrontEnd/New_FrontEnd/UI_FN_Button_Hover_Click_02}}
]
properties {
 maxDistance=2097152
 soundClassObject=&Fort_UI
 volumeMultiplier=1
 concurrencyOverrides {bLimitToOwner=true, maxCount=1, resolutionRule=PreventNew, retriggerTime=0.12}
 concurrencySet [/Game/Sounds/Concurrency_Rules/FrontEnd_UI_Menu_Max_01_Stop_Oldest_Hover_Tick]
}
